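; BP_PrimitiveColCompare is a test actor that shows one AtomColorInfo colour delivered to materials in three ways for comparison.
; Its construction script picks a data table row by index and labels it with the row name on a text render.
; The colour feeds a dynamic material vector parameter directly, and also goes bit-packed as a float into custom primitive data and per-instance custom data.
; The BeginPlay, Tick and overlap events are empty.
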(event EventBeginPlay)

(event Collision|EventActorBeginOverlap
  (params (OtherActor :type Actor%20Object%20Reference)))

(event EventTick
  (params (DeltaSeconds :type Float%20%28single-precision%29)))

(fn ConstructionScript
  (bind _returnvalue (Math|Transform|MakeTransform "0, 0, 0"))
  (bind _datatable (Variables|Default|GetDataTable))
  (bind _direct_colour_set (Variables|Default|GetDirect%20Colour%20Set))
  (bind _newvar_1 (Variables|Default|GetNewVar_1))
  (bind _per_instance_custom (Variables|Default|GetPer%20Instance%20Custom))
  (bind _outrownames (DataTable|GetDataTableRowNames _datatable))
  (bind _returnvalue_2 (Utilities|GetDataTableRow _datatable (Utilities|Array|Get(aref) _outrownames _newvar_1)))
  (Utilities|FlowControl|Sequence
    (:then_0
      (bind _returnvalue_3 (Rendering|Material|CreateDynamicMaterialInstance _direct_colour_set 0 "/AtomFramework/Materials/Tools/Tests/MI_AtomColourTest.MI_AtomColourTest" "None" :!class /Script/Engine.PrimitiveComponent))
      (Rendering|Material|SetVectorParameterValue _returnvalue_3 "Color" (Math|Conversions|ToLinearColor(Color) (Utilities|Struct|BreakAtomColorInfo _returnvalue_2)) :!class /Script/Engine.MaterialInstanceDynamic)
      (Components|InstancedStaticMesh|AddInstance _direct_colour_set _returnvalue false))
    (:then_1
      (Rendering|Material|SetCustomPrimitiveDataFloat (Variables|Default|GetCustom%20Prim%20Mesh) 0 (Math|Conversions|ToFloat(Integer) (AtomMaterial|GetBitPackforColor (Utilities|Struct|BreakAtomColorInfo _returnvalue_2)))))
    (:then_2
      (Components|InstancedStaticMesh|AddInstance _per_instance_custom _returnvalue false)
      (Components|InstancedStaticMesh|SetCustomDataValue _per_instance_custom 0 0 (Math|Conversions|ToFloat(Integer) (AtomMaterial|GetBitPackforColor (Utilities|Struct|BreakAtomColorInfo _returnvalue_2))) false))
    (:then_3
      (Rendering|Components|TextRender|SetText (Variables|Default|GetTextRender) (Utilities|Text|ToText(Name) (Utilities|Array|Get(aref) _outrownames _newvar_1))))))
