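; BP_PrimitiveTest is a test actor that previews the Atom colour DataTable as a labelled grid of instances.
; The construction script adds each row to an Opaque, Transparent, Glitter or Opal instanced mesh by MaterialType, labels it and logs its packed integer into INTString.
; SetBrickColour bit-packs RGB into per-instance custom data slot 0.
; UpdateRenderTarget, uncalled here, draws colours as single texels indexed by numeric row name into a nearest-filtered render target.
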(event EventBeginPlay)

(event Collision|EventActorBeginOverlap
  (params (OtherActor :type Actor%20Object%20Reference)))

(event EventTick
  (params (DeltaSeconds :type Float%20%28single-precision%29)))

(fn SetBrickColour
  (params (Instance%20Target :type Instanced%20Static%20Mesh%20Component%20Object%20Reference) (InstanceIndex :type Integer) (InColor :type Linear%20Color%20Structure) (NewParam :type Float%20%28double-precision%29))
  (Components|InstancedStaticMesh|SetCustomDataValue Instance%20Target InstanceIndex 0 (Math|Conversions|ToFloat(Integer) (Utilities|BitPackRGB InColor)) true))

(fn UpdateRenderTarget
  (bind _rt (Variables|Default|GetRT))
  (bind _datatable (Variables|Default|GetDataTable))
  (Variables|/Script/Engine.Texture|SetFilter "TF_Nearest" _rt)
  (Variables|/Script/Engine.Texture|SetSRGB false _rt)
  (Variables|/Script/Engine.Texture|SetCompressionNoAlpha true _rt)
  (bind (_canvas _size _context) (Rendering|BeginDrawCanvastoRenderTarget _rt))
  (bind _outrownames (DataTable|GetDataTableRowNames _datatable))
  (for _array_element _outrownames
    (bind _returnvalue (Utilities|GetDataTableRow _datatable _array_element))
    (Canvas|DrawTexture _canvas "" (Math|Vector2D|MakeVector2D (Math|Conversions|ToFloat(Integer) (Utilities|String|StringToInteger (Utilities|String|ToString(Name) _array_element)))) "(X=1.000000,Y=1.000000)" "(X=0.000000,Y=0.000000)" "(X=1.000000,Y=1.000000)" (Math|Color|MakeColor (Math|Color|BreakColor (Math|Conversions|ToLinearColor(Color) (Utilities|Struct|BreakAtomColorInfo _returnvalue))) (.G (Math|Color|BreakColor (Math|Conversions|ToLinearColor(Color) (Utilities|Struct|BreakAtomColorInfo _returnvalue)))) (.B (Math|Color|BreakColor (Math|Conversions|ToLinearColor(Color) (Utilities|Struct|BreakAtomColorInfo _returnvalue))))) "BLEND_Translucent" 0.0 "(X=0.500000,Y=0.500000)"))
  (Rendering|EndDrawCanvastoRenderTarget _context))

(fn ConstructionScript
  (bind _transparent (Variables|Default|GetTransparent))
  (bind _opaque (Variables|Default|GetOpaque))
  (bind _opal (Variables|Default|GetOpal))
  (bind _count (Variables|Default|GetCount))
  (bind _glitter (Variables|Default|GetGlitter))
  (bind _datatable (Variables|Default|GetDataTable))
  (Variables|/Script/Engine.Texture|SetSRGB true (Variables|Default|GetRT))
  (Variables|Default|SetINTString "")
  (bind _outrownames (DataTable|GetDataTableRowNames _datatable))
  (for (_array_element _array_index) _outrownames
    (bind _returnvalue (Utilities|GetDataTableRow _datatable _array_element))
    (bind _returnvalue_2 (Components|InstancedStaticMesh|AddInstance (Utilities|Select :Index (.MaterialType (Utilities|Struct|BreakAtomColorInfo _returnvalue)) :Standard _opaque :Transparent _transparent :Glitter _glitter :Opalescent _opal) (Math|Transform|MakeTransform (Math|Vector|MakeVector (* (Math|Float|Floor (/ _array_index _count)) 4e+01) (* (Math|Float|Fraction (/ _array_index _count)) -3e+02)) "0, 0, 0" "1.000000,1.000000,1.000000") false))
    (CallFunction|SetBrickColour (Utilities|Select :Index (.MaterialType (Utilities|Struct|BreakAtomColorInfo _returnvalue)) :Standard _opaque :Transparent _transparent :Glitter _glitter :Opalescent _opal) _returnvalue_2 (Math|Conversions|ToLinearColor(Color) (Utilities|Struct|BreakAtomColorInfo _returnvalue)) (Math|Conversions|ToFloat(Integer) (AtomMaterial|GetBitPackforColor (Utilities|Struct|BreakAtomColorInfo _returnvalue))))
    (bind _returnvalue_3 (AddComponent|Rendering|AddTextRenderComponent false (Math|Transform|MakeTransform (+ (.location (Math|Transform|MakeTransform (Math|Vector|MakeVector (* (Math|Float|Floor (/ _array_index _count)) 4e+01) (* (Math|Float|Fraction (/ _array_index _count)) -3e+02)) "0, 0, 0" "1.000000,1.000000,1.000000")) "0,0,15.000000") (.rotation (Math|Transform|MakeTransform (Math|Vector|MakeVector (* (Math|Float|Floor (/ _array_index _count)) 4e+01) (* (Math|Float|Fraction (/ _array_index _count)) -3e+02)) "0, 0, 0" "1.000000,1.000000,1.000000")) (.scale (Math|Transform|MakeTransform (Math|Vector|MakeVector (* (Math|Float|Floor (/ _array_index _count)) 4e+01) (* (Math|Float|Fraction (/ _array_index _count)) -3e+02)) "0, 0, 0" "1.000000,1.000000,1.000000")))))
    (Rendering|Components|TextRender|SetText _returnvalue_3 (Utilities|Text|ToText(Name) _array_element))
    (Variables|Default|SetINTString (Utilities|String|Append (Variables|Default|GetINTString) "
" (Utilities|String|ToString(Integer) (AtomMaterial|GetBitPackforColor (Utilities|Struct|BreakAtomColorInfo _returnvalue)))))))
