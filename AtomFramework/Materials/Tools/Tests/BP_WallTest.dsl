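; BP_WallTest is a test Blueprint that builds a brick wall from instanced static meshes in its construction script.
; It applies a chosen mesh and one of five primitive materials, then loops over height and width counts, offsetting alternate rows by half a brick.
; A seeded random stream sampled through a curve varies each brick's rotation.
; The colour from one selected data table row is bit-packed into a single per-instance custom data float.
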(event EventBeginPlay)

(event Collision|EventActorBeginOverlap
  (params (OtherActor :type Actor%20Object%20Reference)))

(event EventTick
  (params (DeltaSeconds :type Float%20%28single-precision%29)))

(fn ConstructionScript
  (bind _instancedstaticmesh (Variables|Default|GetInstancedStaticMesh))
  (bind (_data_table_datatable _data_table_rowname) (Utilities|Struct|BreakDataTableRowHandle (Variables|Getdata%20table_DataTable)))
  (Components|StaticMesh|SetStaticMesh (Variables|Default|GetInstancedStaticMesh) (Variables|Default|GetMesh))
  (Rendering|Material|SetMaterial (Variables|Default|GetInstancedStaticMesh) 0 (Utilities|Select :Index (Variables|Default|GetMaterialType) :Standard "/AtomFramework/Materials/Game_Qual/MI_Primitive_Standard.MI_Primitive_Standard" :Transparent "/AtomFramework/Materials/Game_Qual/MI_Primitive_Trans.MI_Primitive_Trans" :Glitter "/AtomFramework/Materials/Game_Qual/MI_Primitive_Glit.MI_Primitive_Glit" :Opalescent "/AtomFramework/Materials/Game_Qual/MI_Primitive_Opal.MI_Primitive_Opal" :Metallic "/AtomFramework/Materials/Game_Qual/MI_Primitive_Metal.MI_Primitive_Metal"))
  (for _index (range (+ (Variables|Default|GetHeight%20Count) 1))
    (for _index_2 (range (+ (Variables|Default|GetWidth%20Count) 1))
      (bind _returnvalue (Components|InstancedStaticMesh|AddInstance _instancedstaticmesh (Math|Transform|MakeTransform (Math|Vector|MakeVector (+ (* (Math|Conversions|ToFloat(Integer) _index_2) 31.9) (* (Math|Float|Fraction (/ (Math|Conversions|ToFloat(Integer) _index) 2.0)) 31.9)) 0.0 (* (Math|Conversions|ToFloat(Integer) _index) 9.62)) (Math|Rotator|MakeRotator 0.0 0.0 (* (Math|Curves|GetFloatValue (Variables|Default|GetRotCurve) (Math|Random|RandomFloatinRangefromStream (Variables|Default|GetStream) -1.0 1.0)) (Variables|Default|GetRotationScale))) "1.000000,1.000000,1.000000") false))
      (bind _outrownames (DataTable|GetDataTableRowNames _data_table_datatable))
      (bind _returnvalue_2 (Utilities|GetDataTableRow _data_table_datatable _data_table_rowname))
      (Components|InstancedStaticMesh|SetCustomDataValue _instancedstaticmesh _returnvalue 0 (Math|Conversions|ToFloat(Integer) (Utilities|BitPackRGB (Utilities|Struct|BreakSTBrickCol _returnvalue_2))) false))))
